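# Translucent unlit hologram material for the PlotTwist RiftGate Noble character, with a TurnOn parameter that drives a vertical pulse.
# Flickering, panning world-aligned noise and the pulse feed vertex offset and a Fresnel, gradient and lighting intensity.
# That intensity samples a curve-atlas color row, modulated by optional diffuse overlay, highlight scanline and time of day.
# Opacity combines scanlines, depth fade and a biased custom-depth comparison that culls hidden layers, with optional separate head/eye bias.
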Material /PlotTwist/Items/RiftGate/Assets/Characters/Noble/Materials/M_PlotTwist_RiftGate_Hologram
graph {
 properties {
  allowTranslucentCustomDepthWrites=true
  bEnableResponsiveAA=true
  bUsedWithMeshDeformer=true
  bUsedWithNiagaraMeshParticles=true
  bUsedWithNiagaraSprites=true
  bUsedWithSkeletalMesh=true
  blendMode=BLEND_Translucent
  opacityMaskClipValue=0
  pixelDepthOffsetMode=PDOM_Legacy
  shadingModel=MSM_Unlit
  usedShadingModels=Unlit
 }
 nodes [
  n1 = ScalarParameter() {defaultValue=0.6, parameterName=MinScanlineOpacity}
  n2 = VectorParameter() {group=Noise, parameterName=NoiseTileMult, defaultValue=<rgba 1 1 0.1 1>}
  n3 = Divide(A=ObjectRadius(), B=ScalarParameter() {defaultValue=0.1, group=Noise, parameterName=NoiseSize})
  n4 = WorldPosition()
  n5 = Multiply(A=@n3, B=Multiply(A=Time(), B=VectorParameter() {group=Noise, parameterName=NoisePanSpeed, defaultValue=<rgba -0.08 0.07 -0.2 0>}.RGB))
  n6 = Power(Base=MaterialFunctionCall(
    TextureObject=TextureObject() {samplerType=SAMPLERTYPE_LinearColor, texture=&T_NoisePack_LinearColor}
    TextureSize=Multiply(A=@n2.RGB, B=@n3)
    WorldPosition=Add(A=@n4.XYZ, B=@n5)
   ) {materialFunction=&WorldAlignedTexture}.XY Texture) {constExponent=4}
  n7 = ScalarParameter() {defaultValue=2, group=Noise, parameterName=NoiseMaskMult}
  n8 = ScalarParameter() {group=Noise, parameterName=NoiseAdd}
  n9 = ActorPositionWS()
  n10 = MaterialFunctionCall(Float2=Round(None=Divide(A=@n9) {constB=2000})) {materialFunction=&BreakOutFloat2Components}
  n11 = Add(A=Time(), B=Add(A=@n10.R, B=@n10.R))
  n12 = Multiply(A=ScalarParameter() {defaultValue=0.5, parameterName=FlickerSpeed}, B=@n11)
  n13 = AppendVector(A=AppendVector(A=Sine(None=@n12) {period=1.6}, B=Sine(None=@n12) {period=1.42}), B=Sine(None=@n12) {period=1.51})
  n14 = ScalarParameter() {defaultValue=0.5, group=Noise, parameterName=NoiseFrequency}
  n15 = Add(A=Sine(None=@n12) {period=2.1}, B=Multiply(B=@n14) {constA=0.25})
  n16 = Add(A=Sine(None=@n12) {period=0.72}, B=Multiply(B=@n14) {constA=0.75}) {constB=0.75}
  n17 = Ceil(None=Add(A=Multiply(B=@n14) {constA=0.9, constB=0.9}, B=Sine(None=@n12) {period=1.25}) {constB=0.9})
  n18 = Saturate(None=Multiply(A=Multiply(A=@n15, B=@n16), B=Multiply(A=@n16, B=@n17)))
  n19 = LinearInterpolate(Alpha=ConstantBiasScale(None=Sine(None=@n11) {period=0.075})) {constA=0.7}
  n20 = Multiply(A=Multiply(A=Ceil(None=@n13), B=Multiply(A=@n18, B=@n19)), B=ScalarParameter() {defaultValue=2, group=Noise, parameterName=NoiseMult})
  n21 = Saturate(None=Multiply(A=Multiply(A=@n6, B=@n7) {constB=20}, B=Add(A=@n8, B=@n20)))
  n22 = WorldPosition()
  n23 = ComponentMask(None=MaterialFunctionCall() {materialFunction=&ObjectPivotPoint}.Object Pivot Location) {b=true}
  n24 = Add(A=@n23, B=ScalarParameter() {defaultValue=300, parameterName=PulseHeight}) {constB=1600}
  n25 = ScalarParameter() {defaultValue=1, parameterName=TurnOn}
  n26 = ScalarParameter() {defaultValue=33, parameterName=PulseRadius}
  n27 = Multiply(A=@n26) {constB=0.1}
  n28 = Multiply(
   A=SphereMask(
    A=ComponentMask(None=@n22.XYZ) {b=true}
    B=LinearInterpolate(A=@n23, B=@n24, Alpha=@n25) {constA=-0.25, constB=1.2}
    Radius=LinearInterpolate(A=@n26, B=@n27, Alpha=@n25) {constA=500, constB=100}
   ) {hardnessPercent=0}
   B=OneMinus(None=@n25)
  )
  n29 = Add(A=Desaturation(None=NamedRerouteDeclaration(None=@n21) {name=Noise, nodeColor=<rgba 1 0.7734 0 1>}), B=Multiply(A=@n28) {constB=2})
  n30 = QualitySwitch(
   Default=MaterialFunctionCall(In=@n29, Contrast=ScalarParameter() {defaultValue=0.2, group=Noise, parameterName=NoiseContrast}) {materialFunction=&CheapContrast}.Result
   Low=Constant()
  )
  n31 = Multiply(A=@n30) {constB=1.5}
  n32 = Saturate(None=@n30)
  n33 = Multiply(A=ScalarParameter() {defaultValue=200, parameterName=ScanlineTile}, B=LinearInterpolate(Alpha=Ceil(None=@n32)) {constA=1, constB=1.25})
  n34 = ScalarParameter() {defaultValue=20000, parameterName=ScaleDistance}
  n35 = Divide(A=Distance(A=CameraPositionWS(), B=ActorPositionWS()), B=@n34) {constB=1000}
  n36 = Divide(A=ComponentMask(None=WorldPosition().XYZ) {b=true}, B=LinearInterpolate(B=@n34, Alpha=Saturate(None=@n35)) {constA=50, constB=100}) {constB=20}
  n37 = ConstantBiasScale(None=Sine(None=Multiply(A=@n33, B=@n36) {constA=1}))
  n38 = VertexNormalWS()
  n39 = Constant3Vector() {constant=<rgba 0 0 1 0>}
  n40 = Power(Base=Abs(None=DotProduct(A=@n38, B=@n39))) {constExponent=5}
  LinearInterpolate(
   A=Clamp(None=Subtract(A=@n1, B=@n31)) {minDefault=-0.25}
   B=Add(A=@n31)
   Alpha=Power(Base=LinearInterpolate(A=@n37, Alpha=@n40) {constB=0.25}) {constExponent=1.5}
  ) {constA=0.5}
  n41 = Subtract(
   A=WorldPosition().XYZ
   B=MaterialFunctionCall() {materialFunction=&ObjectPivotPoint}.Object Pivot Location
  )
  n42 = LinearInterpolate(Alpha=ScalarParameter() {defaultValue=1, parameterName=TurnOn}) {constA=-1, constB=0}
  n43 = Normalize(VectorInput=Subtract(A=WorldPosition().XYZ, B=ObjectPositionWS()))
  n44 = Constant3Vector() {constant=<rgba 1 1 0 0>}
  n45 = ScalarParameter() {defaultValue=10, group=Noise, parameterName=WPOStrength}
  n46 = Multiply(A=ObjectRadius(), B=ScalarParameter() {defaultValue=0.025, group=Noise, parameterName=WPOStrengthFromObjectRadius}) {constB=0.05}
  n47 = Multiply(A=@n29, B=Multiply(A=Multiply(A=@n43, B=@n44), B=Add(A=@n45, B=@n46))) {constB=0}
  LinearInterpolate(A=@n47, B=Add(A=@n47, B=Multiply(A=@n41, B=@n42)), Alpha=ScalarParameter() {group=Opaque, parameterName=Opaque Material}) {constB=0}
  n48 = VectorParameter() {
   desc="Time of Day Emissive multiplier.  \r\nR - Morning\r\nG - Day\r\nB - Evening\r\nA - Night"
   group=Lighting
   parameterName="TOD Emissive Multiplier (M,D,E,N)"
   defaultValue=<rgba 0.75 1 0.75 0.5>
  }
  n49 = VectorParameter() {parameterName=ColorMult, defaultValue=<rgba 1 1 1 0>}
  n50 = Normalize(VectorInput=VectorParameter() {parameterName=LightDirection, defaultValue=<rgba 0 0 1 1>}.RGB)
  n51 = VertexNormalWS()
  n52 = TextureSampleParameter2D() {group=Texture, parameterName=Normals, samplerType=SAMPLERTYPE_Normal, texture=/Engine/EngineMaterials/BaseFlattenNormalMap}
  n53 = Transform(None=@n52.RGB)
  n54 = ScalarParameter() {group=Lighting, parameterName=UseNormal}
  n55 = MaterialFunctionCall(
   Input=DotProduct(A=Transform(None=@n50) {transformSourceType=TRANSFORMSOURCE_Local}, B=LinearInterpolate(A=@n51, B=@n53, Alpha=@n54))
   Input Low=Constant() {r=-1}
   Input High=Constant() {r=1}
   Target Low=ScalarParameter() {defaultValue=-1, parameterName=LightDotLow}
   Target High=ScalarParameter() {defaultValue=1, parameterName=LightDotHigh}
  ) {materialFunction=&RemapValueRange}
  n56 = ScalarParameter() {defaultValue=0.2, group=Lighting, parameterName=NormalBrightnessAddMin}
  n57 = ScalarParameter() {defaultValue=0.3, group=Lighting, parameterName=NormalBrightnessAddMax}
  n58 = NamedRerouteUsage()
  n59 = TextureSampleParameter2D() {group=Texture, parameterName=M, texture=/Engine/EngineResources/WhiteSquareTexture}
  n60 = LinearInterpolate(A=ScalarParameter() {defaultValue=1, parameterName=TurnOn}, Alpha=ScalarParameter() {group=Opaque, parameterName=Opaque Material})
  n61 = LinearInterpolate(B=ScalarParameter() {defaultValue=0.6, group=Noise, parameterName=NoiseFresnelXPMult}, Alpha=@n30) {constA=1, constB=0.5}
  n62 = Multiply(A=ScalarParameter() {defaultValue=1, parameterName=FresnelExp}, B=@n61)
  n63 = ScalarParameter() {defaultValue=0.5, parameterName=FresnelMin}
  n64 = PixelNormalWS()
  n65 = ScalarParameter() {defaultValue=1, parameterName=EdgeFresnelXP}
  n66 = LinearInterpolate(Alpha=Fresnel(ExponentIn=Multiply(A=@n61, B=@n65), Normal=@n64) {baseReflectFraction=0, exponent=1}) {constA=5, constB=-0.4}
  n67 = Power(Base=Multiply(A=Fresnel(ExponentIn=@n62, BaseReflectFractionIn=@n63, Normal=@n64) {baseReflectFraction=0, exponent=0.5}, B=Saturate(None=@n66)))
  n68 = Multiply(A=ScalarParameter() {defaultValue=-0.4, group=Noise, parameterName=NoiseBrightness}, B=@n32) {constB=0.1}
  n69 = MaterialFunctionCall!(
   Number of Scanlines=ScalarParameter() {defaultValue=150, group=Scanlines, parameterName=Scanline Amount}
   Scanline Speed=ScalarParameter() {defaultValue=0.009, group=Scanlines, parameterName=Scanline Speed}
   "Scanlines Thickness (Bias)"=ScalarParameter() {defaultValue=-0.065, group=Scanlines, parameterName=ScanlineSize}
   Scanlines Dark Value=ScalarParameter() {defaultValue=0.2, group=Scanlines, parameterName=Scanlines Dark}
   Scanlines Light Value=ScalarParameter() {defaultValue=0.85, group=Scanlines, parameterName=Scanlines Light}
   Block Opacity
   UVs
   Time=Multiply(A=Time(), B=ScalarParameter() {defaultValue=0.1, group=Scanlines, parameterName=ScanlinesTimeMultiplier})
   Block Repeat
  ) {materialFunction=/Game/Effects/Fort_Effects/Effects/Atlas/Materials/MF_Atlas_Scanlines_01}
  n70 = Add(A=ScalarParameter() {defaultValue=0.7, parameterName=ScanlineColorMinInfluence}, B=@n69.Result)
  n71 = ComponentMask(None=Subtract(A=WorldPosition().XYZ, B=ObjectPositionWS())) {b=true}
  n72 = ObjectRadius()
  n73 = Saturate(None=ConstantBiasScale(None=Divide(A=@n71, B=@n72)))
  n74 = LinearInterpolate(
   A=Multiply(A=Add(A=@n67, B=@n68), B=Saturate(None=@n70))
   B=LinearInterpolate(A=@n73, B=OneMinus(None=@n73), Alpha=ScalarParameter() {parameterName=FlipGradient})
   Alpha=ScalarParameter() {defaultValue=0.4, parameterName=VerticalGradientInflience}
  )
  n75 = Multiply(
   A=LinearInterpolate(
    A=ScalarParameter() {parameterName=MaskMin}
    Alpha=ChannelMaskParameter(None=@n59.RGB) {maskChannel=Green, parameterName=MaskChannel, defaultValue=<rgba 0 1 0 0>}.R
   )
   B=Add(A=Multiply(A=@n60, B=@n74), B=@n28) {constA=1}
  )
  n76 = Add(A=Multiply(A=Clamp(None=@n55.Result), B=LinearInterpolate(A=@n56, B=@n57, Alpha=@n58)) {constB=0.15}, B=@n75)
  n77 = Constant()
  n78 = ScalarParameter() {defaultValue=1, parameterName=LightMaxReduction}
  n79 = Constant() {r=1}
  n80 = CurveAtlasRowParameter(CurveTime=Saturate(None=MaterialFunctionCall(Input=@n76, Input Low=@n77, Input High=@n78, Target Low=@n77, Target High=@n79) {materialFunction=&RemapValueRange}.Result)) {atlas=&AthenaCurveAtlas01, curve=/Game/Athena/LinearColorCurves/Curve_Galileo_Hologram, defaultValue=171}
  n81 = TextureSampleParameter2D() {group=Diffuse Texture Overlay, parameterName=Diffuse Texture Overlay, texture=&WhiteSquareTexture}
  n82 = ScalarParameter() {defaultValue=1, group=Opaque, parameterName=Diffuse Multi}
  n83 = LinearInterpolate(
   A=LinearInterpolate(
    A=@n81.RGB
    B=MaterialFunctionCall!(Base=@n81.RGB, Add=@n82) {materialFunction=/Engine/Functions/Engine_MaterialFunctions02/Math/MultiplyAdd}.Result
    Alpha=ScalarParameter() {group=Opaque, parameterName=Opaque Material}
   )
   B=MaterialFunctionCall(In=@n81.RGB, Contrast=ScalarParameter() {group=Diffuse Texture Overlay, parameterName=Diffuse Contribution Cheap Contrast}) {materialFunction=&CheapContrast}.Result
   Alpha=ScalarParameter() {group=Diffuse Texture Overlay, parameterName=Diffuse Overlay Use Cheap Contrast}
  )
  n84 = ScalarParameter() {defaultValue=1, group=Diffuse Texture Overlay, parameterName=Diffuse Contribution Desaturation}
  n85 = LinearInterpolate(
   A=@n80
   B=Multiply(A=Desaturation(None=@n83, Fraction=@n84), B=@n80)
   Alpha=ScalarParameter() {group=Diffuse Texture Overlay, parameterName=Lerp Diffuse Overlay}
  )
  n86 = Divide(A=ObjectRadius(), B=ScalarParameter() {defaultValue=1.5, group=Highlight Scanline, parameterName=HIghlight Scanline Size})
  n87 = Multiply(A=Time(), B=VectorParameter() {group=Highlight Scanline, parameterName=Highlight Scanline Panner, defaultValue=<rgba 0 0 2 1>}.RGB)
  n88 = MaterialFunctionCall(
   TextureObject=TextureObject() {texture=/Game/Global/Textures/Default/Masks_and_Gradients/gradient_stasis}
   TextureSize=Multiply(A=VectorParameter() {group=Highlight Scanline, parameterName=Highlight Scanline Size, defaultValue=<rgba 1 2 1 1>}.RGB, B=@n86)
   WorldPosition=Add(A=WorldPosition().XYZ, B=Multiply(A=@n86, B=@n87))
  ) {materialFunction=&WorldAlignedTexture}
  n89 = Multiply(
   A=LinearInterpolate(A=@n85, B=Multiply(A=@n85, B=@n69.Result), Alpha=ScalarParameter() {group=Opaque, parameterName=Opaque Material})
   B=LinearInterpolate(
    A=ScalarParameter() {defaultValue=2, group=Highlight Scanline, parameterName=EmissiveMultMin}
    B=ScalarParameter() {defaultValue=2.5, group=Highlight Scanline, parameterName=EmissiveMultMax}
    Alpha=NamedRerouteDeclaration(None=@n88.XY Texture) {name=Highlight Scanline, nodeColor=<rgba 0 1 0.0625 1>}
   )
  ) {constB=2}
  MaterialFunctionCall(Emissive=Multiply(
    A=MaterialFunctionCall(Value_Morning=@n48.R, Value_Day=@n48.R, Value_Evening=@n48.R, Value_Night=@n48.R) {materialFunction=&Scalar_Time_of_Day}.Result
    B=Multiply(A=@n49.RGB, B=@n89)
   )) {materialFunction=&MF_FX_EmissiveAdjustments}
  n90 = ScalarParameter() {defaultValue=5, parameterName=FadeDistance}
  n91 = Saturate(None=Multiply(A=DepthFade(Opacity=@n75, FadeDistance=@n90), B=@n69.Result) {constA=1})
  n92 = ScalarParameter() {defaultValue=1.25, parameterName=OpacityMult}
  n93 = CameraPositionWS()
  n94 = ActorPositionWS()
  n95 = Saturate(None=Divide(A=Distance(A=@n93, B=@n94)) {constB=1000})
  n96 = LinearInterpolate(
   A=ScalarParameter() {defaultValue=-0.25, parameterName=Bias_Close, sliderMax=1, sliderMin=-1}
   B=ScalarParameter() {defaultValue=-1, parameterName=Bias_Far, sliderMax=1, sliderMin=-1}
   Alpha=@n95
  )
  n97 = ScalarParameter() {defaultValue=-0.25, desc="Only used if sepeate head and eyes bias = 1", parameterName=Bias_Close_EyesMouth, sliderMax=1, sliderMin=-1}
  n98 = ScalarParameter() {defaultValue=-1, desc="Only used if sepeate head and eyes bias = 1", parameterName=Bias_Far_EyesMouth, sliderMax=1, sliderMin=-1}
  n99 = LinearInterpolate(A=@n97, B=@n98, Alpha=@n95)
  n100 = MaterialFunctionCall(Float2=TextureCoordinate()) {materialFunction=&BreakOutFloat2Components}
  n101 = VectorParameter() {
   desc="Square mask in UV Space.\r\n\r\nOnly used if sepeate head and eyes bias = 1"
   parameterName=BiasHeadEyesMaskPos
   channelNames {a=BottomRightPointY, b=BottomRightPointX, g=TopLeftPointY, r=TopLeftPointX}
   defaultValue=<rgba 0.2 0 0.8 1>
  }
  n102 = OneMinus(None=@n101.TopLeftPointX)
  n103 = Floor(None=Saturate(None=Add(A=@n100.R, B=@n102) {constB=0.25}))
  n104 = OneMinus(None=@n101.TopLeftPointY)
  n105 = Floor(None=Saturate(None=Add(A=@n100.G, B=@n104) {constB=0.25}))
  n106 = Ceil(None=Saturate(None=Subtract(A=@n101.BottomRightPointX, B=@n100.R)))
  n107 = Ceil(None=Saturate(None=Subtract(A=@n101.BottomRightPointY, B=@n100.G)))
  n108 = Saturate(None=Min(A=Min(A=@n103, B=@n105), B=Min(A=@n106, B=@n107)))
  n109 = ScalarParameter() {parameterName=UseSeperateHeadAndEyesBias, sliderMax=1, sortPriority=31}
  n110 = Add(A=PixelDepth().R, B=LinearInterpolate(A=@n96, B=LinearInterpolate(A=@n99, B=@n96, Alpha=@n108), Alpha=Round(None=@n109)))
  n111 = ComponentMask(None=SceneTexture() {sceneTextureId=PPI_CustomDepth}.Color) {r=true}
  n112 = Constant()
  n113 = Constant() {r=1}
  n114 = Subtract(A=WorldPosition().XYZ, B=ObjectPositionWS())
  n115 = ConstantBiasScale(None=Divide(A=ComponentMask(None=@n114) {b=true}, B=ObjectBounds()))
  Multiply(
   A=Multiply(A=Multiply(A=@n91, B=@n92), B=If(A=@n110, B=@n111, "A > B"=@n112, "A == B"=@n112, "A < B"=@n113))
   B=Multiply(A=Saturate(None=@n115), B=ScalarParameter() {defaultValue=1, parameterName=Object Z Opacity Multi})
  )
  n116 = MaterialFunctionCall(Offset Amount=ScalarParameter() {parameterName=Camera Offset}) {materialFunction=/Engine/Functions/Engine_MaterialFunctions02/WorldPositionOffset/CameraOffset}
  Reroute(None=Add(
    A=MaterialFunctionCall!(WPO) {materialFunction=/Game/Packages/Fortress_SharedMaterials/Base_Material_Functions/Fortnite_HideElementsWithVertexColor}.Result
    B=Add(A=@n116.Clamped Camera Offset, B=@n47)
   ))
  Saturate(None=@n76)
  n117 = ComponentMask(None=TextureCoordinate()) {r=true}
  n118 = OneMinus(None=@n117)
  n119 = Saturate(None=Multiply(A=Multiply(A=@n118, B=@n117)) {constB=2})
  n120 = ScalarParameter() {
   defaultValue=0.3
   desc="eyes and mouth are usually located on left/right of the texture. Lower values increase edge padding/thickness.\r\n\r\nOnly used if sepeate head and eyes bias = 1"
   parameterName=BiasHeadEyesMaskPadding
   sliderMax=0.5
  }
  Saturate(None=Round(None=Add(A=@n119, B=@n120) {constB=0.25}))
  ScalarParameter() {defaultValue=-0.25, parameterName=BiasSorting_Far}
  VectorParameter() {
   desc="Square mask in UV Space.\r\n\r\nOnly used if sepeate head and eyes bias = 1"
   parameterName=BiasHeadEyesMaskPos
   channelNames {a=BottomRightPointY, b=BottomRightPointX, g=TopLeftPointY, r=TopLeftPointX}
   defaultValue=<rgba 0.2 0 0.8 1>
  }
 ]
}
